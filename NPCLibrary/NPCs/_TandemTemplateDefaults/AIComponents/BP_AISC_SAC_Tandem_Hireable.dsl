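; Special actor component for hireable Tandem NPCs that sets their minimap and compass icon brushes.
; On server activation, if the adjust-for-NoHire option is on, it checks the owner for the NPC.Status.NoHire gameplay tag.
; Tagged owners get the speech brushes and untagged owners get the hireable brushes, each with a separate compass variant.
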(event EventOnServerActivated
  (if (Variables|Default|GetAdjustBrushForNoHireTag)
    (bind _returnvalue (GameplayTags|HasMatchingGameplayTag(Message) (Components|GetOwner) "(TagName=\"NPC.Status.NoHire\")"))
    (if _returnvalue
      (Variables|Minimap|SetSpecialActorMinimapIconBrush (Variables|Default|GetSpeechSlateBrush))
      (Variables|Compass|SetSpecialActorCompassIconBrush (Variables|Default|GetSpeechSlateBrush_Compass))
      (else
        (Variables|Minimap|SetSpecialActorMinimapIconBrush (Variables|Default|GetHIreableSlateBrush))
        (Variables|Compass|SetSpecialActorCompassIconBrush (Variables|Default|GetHIreableSlateBrush_Compass))))))

(event EventBeginPlay
  (|CallParentFunction))

(event EventTick
  (params (DeltaSeconds :type Float%20%28single-precision%29))
  (|CallParentFunction DeltaSeconds) ; This node is disabled and will not be called.\nDrag off pins to build functionality.
)
